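# Niagara trail effect for the Storm Aviator pickaxe, built from two ribbon anim-trail emitters, FireTrail and SmokeTrail, that follow the skeletal mesh between two bone or socket indices.
# FireTrail is a short-lived ribbon with added velocity, curve-driven color and randomized dynamic material parameters.
# SmokeTrail lasts longer with two ribbon renderers, deriving extra width and soft-smoke color attributes from its particles.
# Both use variant color selectors, with smoke also enabling a time-of-day curve.
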NiagaraSystem /BRCosmetics/Weapons/FORT_Melee/Pickaxe_StormAviator/FX/NS_Pickaxe_StormAviator_Trail
dependencies {
 usedDataInterfaces [&NiagaraDataInterfaceCurve, &NiagaraDataInterfaceColorCurve, &NiagaraDataInterfaceSkeletalMesh]
 usedDynamicInputs [&Frac_Float]
 usedModules [
  &self:SystemSpawnScript.SystemScriptSource.SystemScriptGraph.NiagaraNodeAssignment_1.SetVariables_#1
  &self:SystemSpawnScript.SystemScriptSource.SystemScriptGraph.NiagaraNodeAssignment_0.SetVariables_#2
  &SystemState
  /BRCosmetics/Characters/Player/Female/Medium/Bodies/F_MED_PiperShelf_Pearl/FX/TimeOfDayGlobal
  &FN_VariantManager
 ]
}
emitters {
 FireTrail {
  emitterSpawnScript [&FN_RibbonIDOffsetUtility()]
  emitterUpdateScript [
   &EmitterState()
   &FN_SpawnAnimTrail()
  ]
  particleSpawnScript [
   &FN_InitializeAnimTrailBoneIndecies(First Bone or Socket Index=1, Second Bone or Socket Index=0)
   &FN_InitializeAnimTrail(AnimTrailWidth=2, Lifetime=0.075, SkeletalMesh=$System.Skeletal Mesh)
   &AddVelocity(Velocity=<0 200 0>)
  ]
  particleUpdateScript [
   &SolveForcesAndVelocity(Clamp Velocity=false, Limit Acceleration=false)
   &FN_UpdateAnimTrail(SkeletalMesh=$System.Skeletal Mesh) {module=FN_UpdateAnimTrail001}
   &DynamicMaterialParameters(
    Index 0 Param 1=&RandomRangeFloat(Minimum=-0.25, Maximum=0.25)
    Index 0 Param 2=&FloatFromCurve(FloatCurve=DataInterface() {propertyValues {Curve {keys [<at 0 1>,<at 1 0.55>]}}})
    Module.Param0WriteEnabled=true
    Module.Param1WriteEnabled=true
    Module.Param2WriteEnabled=false
    Module.Param3WriteEnabled=false
   )
   &ScaleColor(
    Scale RGB=&MakeVectorFromLinearColorRGB(Color=&VariantColorSelector(
      Variant Amount=0
      Original Color=&ColorFromCurve(ColorCurve=DataInterface() {propertyValues {
         AlphaCurve {keys [{arriveTangent=-3.2415, interpMode=RCIM_Cubic, leaveTangent=-3.2415, tangentMode=RCTM_User, time=0.0615, value=1.5}, <at 1 0 RCIM_Cubic>]}
         BlueCurve {keys [<at 0.0075 0>,<at 0.1231 0 RCIM_Cubic>,<at 0.3477 0>,<at 0.5644 0>]}
         GreenCurve {keys [
           <at 0.0075 0.9302>
           {arriveTangent=-2.4862, interpMode=RCIM_Cubic, leaveTangent=-2.4862, time=0.1231, value=0.754}
           <at 0.3477 0.0844>
           <at 0.5644 0>
          ]}
         RedCurve {keys [<at 0.0075 3>, {arriveTangent=-8.1601, interpMode=RCIM_Cubic, leaveTangent=-8.1601, time=0.1231, value=2}, <at 0.3477 0.224>, <at 0.5644 0>]}
        }})
      Variant 1 Color=&ColorFromCurve(ColorCurve=DataInterface() {propertyValues {
         AlphaCurve {keys [{arriveTangent=-3.2415, interpMode=RCIM_Cubic, leaveTangent=-3.2415, tangentMode=RCTM_User, time=0.0615, value=1.5}, <at 1 0 RCIM_Cubic>]}
         BlueCurve {keys [<at 0.0075 0>, <at 0.1231 0 RCIM_Cubic>, {interpMode=RCIM_Cubic, leaveTangent=0.0024, time=0.2815, value=0}, <at 0.5644 0.0011>]}
         GreenCurve {keys [
           <at 0.0075 0.9302>
           {arriveTangent=-3.2423, interpMode=RCIM_Cubic, leaveTangent=-3.2423, time=0.1231, value=0.25}
           {interpMode=RCIM_Cubic, leaveTangent=-1.9657, time=0.2815, value=0.0417}
           <at 0.5644 0>
          ]}
         RedCurve {keys [
           <at 0.0075 3>
           {arriveTangent=-9.1222, interpMode=RCIM_Cubic, leaveTangent=-9.1222, time=0.1231, value=3}
           {interpMode=RCIM_Cubic, leaveTangent=-6.2843, time=0.2815, value=0.5}
           <at 0.5644 0.026>
          ]}
        }})
      Variant 2 Color=?
     ))
    ScaleA=true
    ScaleRGB=true
   )
  ]
  rendererClasses [&NiagaraRibbonRendererProperties]
  renderers [{rendererClass=&NiagaraRibbonRendererProperties}]
 }
 SmokeTrail {
  emitterSpawnScript [&FN_RibbonIDOffsetUtility()]
  emitterUpdateScript [
   &EmitterState()
   &FN_SpawnAnimTrail()
  ]
  particleSpawnScript [
   &FN_InitializeAnimTrailBoneIndecies(First Bone or Socket Index=1, Second Bone or Socket Index=0)
   &FN_InitializeAnimTrail(AnimTrailWidth=2.5, Lifetime=0.175, SkeletalMesh=$System.Skeletal Mesh)
  ]
  particleUpdateScript [
   &SolveForcesAndVelocity(Clamp Velocity=false, Limit Acceleration=false)
   &FN_UpdateAnimTrail(SkeletalMesh=$System.Skeletal Mesh) {module=FN_UpdateAnimTrail001}
   &DynamicMaterialParameters(
    Index 0 Param 1=&RandomRangeFloat(Minimum=-0.25, Maximum=0.25)
    Module.Param0WriteEnabled=true
    Module.Param1WriteEnabled=true
    Module.Param2WriteEnabled=false
    Module.Param3WriteEnabled=false
   )
   /SetVariables(Particles.SmokeWidth:NiagaraFloat=&Multiply_Float(A=$Particles.RibbonWidth, B=1.5)) {isSetParameters=true}
   &ScaleColor(
    Scale Mode=::NewEnumerator1
    Scale RGBA=&MakeVector4_FromLinearColor(LinearColor=&VariantColorSelector(
      Variant Amount=0
      BlendAmount=0.5
      Use TODCurve?=true
      Original Color=&ColorFromCurve(ColorCurve=DataInterface() {propertyValues {
         AlphaCurve {keys [<at 0.6601 2 RCIM_Cubic>,<at 0.902 0 RCIM_Cubic>]}
         BlueCurve {keys [
           <at 0.0224 0.0082 RCIM_Cubic>
           {arriveTangent=0.3462, interpMode=RCIM_Cubic, leaveTangent=0.3462, time=0.0745, value=0.1406}
           <at 0.4682 0.1626>
           <at 0.9039 0.18>
          ]}
         GreenCurve {keys [
           <at 0.0224 0.2909 RCIM_Cubic>
           {arriveTangent=-0.2878, interpMode=RCIM_Cubic, leaveTangent=-0.2878, time=0.0745, value=0.1406}
           <at 0.4682 0.1626>
           <at 0.9039 0.18>
          ]}
         RedCurve {keys [
           <at 0.0224 1 RCIM_Cubic>
           {arriveTangent=-1.7942, interpMode=RCIM_Cubic, leaveTangent=-1.7942, time=0.0745, value=0.3333}
           <at 0.4682 0.2>
           <at 0.9039 0.18>
          ]}
        }})
      Variant 1 Color=&ColorFromCurve(ColorCurve=DataInterface() {propertyValues {
         AlphaCurve {keys [<at 0.6696 1 RCIM_Cubic>,<at 0.9134 0 RCIM_Cubic>]}
         BlueCurve {keys [
           <at 0.0224 0 RCIM_Cubic>
           {arriveTangent=0.1804, interpMode=RCIM_Cubic, leaveTangent=0.1804, time=0.0745, value=0.1406}
           <at 0.4682 0.0805>
           <at 0.9039 0.07>
          ]}
         GreenCurve {keys [
           <at 0.0224 0.0833 RCIM_Cubic>
           {arriveTangent=-0.0065, interpMode=RCIM_Cubic, leaveTangent=-0.0065, time=0.0745, value=0.1406}
           <at 0.4682 0.0805>
           <at 0.9039 0.07>
          ]}
         RedCurve {keys [
           <at 0.0224 1 RCIM_Cubic>
           {arriveTangent=-2.0208, interpMode=RCIM_Cubic, leaveTangent=-2.0208, time=0.0745, value=0.3333}
           <at 0.4682 0.099>
           <at 0.9039 0.07>
          ]}
        }})
      Variant 2 Color=?
     ))
    ScaleRGBA=true
   )
   /SetVariables(Particles.SoftsmokeColor:LinearColor=&MakeLinearColorFromVectorAndFloat(
     "Vector (RGB)"=&MakeVectorFromLinearColorRGB(Color=$Particles.Color)
     "Float (Alpha)"=&FloatFromCurve(FloatCurve=DataInterface() {propertyValues {Curve {keys [{interpMode=RCIM_Cubic, tangentMode=RCTM_User, time=0, value=1}, {interpMode=RCIM_Cubic, tangentMode=RCTM_User, time=0.8, value=0}]}}})
    )) {isSetParameters=true}
  ]
  rendererClasses [&NiagaraRibbonRendererProperties]
  renderers [
   {rendererClass=&NiagaraRibbonRendererProperties}
   {rendererClass=&NiagaraRibbonRendererProperties, rendererIndex=1}
  ]
 }
}
summary {
 systemName=NS_Pickaxe_StormAviator_Trail
 emitters [
  {emitterName=FireTrail, rendererClasses [&NiagaraRibbonRendererProperties]}
  {emitterName=SmokeTrail, rendererClasses [&NiagaraRibbonRendererProperties]}
 ]
 userVariables [
  User.bIsFrontEnd {type=NiagaraBool}
  User.bIsFrontEndPreview {type=NiagaraBool}
  User.ForwardDot {type=NiagaraFloat, defaultValue=0, description=""}
  User.RibbonIDOffset {type=NiagaraInt32}
  User.Variant {type=NiagaraFloat, defaultValue=0, description=""}
  User.WeakPointHitFade {type=NiagaraFloat, defaultValue=0, description=""}
 ]
}
