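# Physics asset for the Storm Aviator pickaxe axe mesh, a two-handed melee cosmetic weapon.
# It holds a single capsule body on the Melee_TwoHanded_Handle bone, offset and rotated along the weapon to approximate its volume.
# There are no constraints, so the weapon is represented as one rigid collision shape with no articulated parts.
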PhysicsAsset /BRCosmetics/Weapons/FORT_Melee/Pickaxe_StormAviator/Meshes/StormAviator_Axe_PhysicsAsset
bodies {Melee_TwoHanded_Handle [{shapeType=Capsule, length=75.4996, radius=36.538, center=<-0.0194 9.0021 31.0891>, rotation=<rot -3.7631 -90 0.0154>}]}
constraints []
